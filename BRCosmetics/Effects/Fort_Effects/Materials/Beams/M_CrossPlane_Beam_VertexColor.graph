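# Unlit, additive, two-sided beam material for Niagara mesh particles, named for cross-plane beam effects.
# Vertex color red blends between low and high color parameters, both red by default, feeding an emissive adjustments function.
# A SplineThicken call uses one constant for base and tip width.
# A separate gradient chain shapes a sine of U, multiplied by V, through a power node; no material output hookups are shown.
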Material /BRCosmetics/Effects/Fort_Effects/Materials/Beams/M_CrossPlane_Beam_VertexColor
graph {
 properties {
  bUsedWithNiagaraMeshParticles=true
  blendMode=BLEND_Additive
  pixelDepthOffsetMode=PDOM_Legacy
  shadingModel=MSM_Unlit
  twoSided=true
  usedShadingModels=Unlit
 }
 nodes [
  n1 = Constant() {r=3}
  MaterialFunctionCall(WidthBase=@n1, WidthTip=@n1) {materialFunction=&SplineThicken}
  MaterialFunctionCall(Emissive=LinearInterpolate(
    A=VectorParameter() {parameterName=LowColor, defaultValue=<rgba 0.1 0 0 1>}.RGB
    B=VectorParameter() {parameterName=HighColor, defaultValue=<rgba 3 0 0 1>}.RGB
    Alpha=VertexColor().R
   )) {materialFunction=&MF_FX_EmissiveAdjustments}
  n2 = MaterialFunctionCall() {materialFunction=&LinearGradient}
  n3 = Sine(None=Divide(A=@n2.UGradient))
  Power(Base=Multiply(A=Saturate(None=@n3), B=@n2.VGradient))
 ]
}
